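; GhostManager_BP spawns ghosts to haunt players in a spawn area during the Puppeteer Demon boss fight.
; Demon spawn starts a looping timer with a once-randomized hotfixable cooldown, cleared on the demon's death.
; Each tick scans the area for players not haunted, on cooldown or downed, pruning stale haunts, then runs an EQS location query.
; The result places a point provider given player, demon and area; spawned ghosts are tracked in arrays parallel to haunted players.
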(event EventBeginPlay
  (Networking|SwitchHasAuthority
    (:Authority
      (Utilities|IsValid (Variables|Default|GetGhostSpawnArea)
        (:Is%20Not%20Valid
          (Development|PrintString "ERROR:: GhostManager_BP:: GhostSpawnArea is missing or invalid."))))))

(event Custom|PuppeteerDemonSpawned
  (params (DemonPawn :type BP%20Player%20Pawn%20Puppeteer%20Demon%20Object%20Reference))
  (Utilities|IsValid DemonPawn
    (:Is%20Valid
      (bind _output_get (Variables|Default|SetDemonPawn DemonPawn))
      (Default|BindEventtoOnHealthChange _output_get (EventDispatchers|CreateEvent :!event Custom|DemonPawnHealthChanged))
      (EventDispatchers|BindEventtoOnDied _output_get (EventDispatchers|CreateEvent :!event Custom|DemonPawnDied))
      (bind _returnvalue (Utilities|Time|SetTimerbyEvent (EventDispatchers|CreateEvent :!event Custom|TryToSpawnGhost) (Math|Random|RandomFloatinRange (ScalableFloat|GetValueatLevel (Variables|Hotfix|GetMinGhostSpawnCooldown_Hf)) (ScalableFloat|GetValueatLevel (Variables|Hotfix|GetMaxGhostSpawnCooldown_Hf))) true))
      (Variables|Default|SetSpawnGhostTimer _returnvalue))))

(event Custom|DemonPawnHealthChanged
  (bind _demonpawn (Variables|Default|GetDemonPawn))
  (Utilities|IsValid _demonpawn
    (:Is%20Valid
      (if (and (FortPawn|IsDead _demonpawn) (GameplayTags|HasMatchingGameplayTag _demonpawn "(TagName=\"PuppeteerBoss.Demon.IsInPhase2\")"))
        (Utilities|Time|ClearandInvalidateTimerbyHandle (Variables|Default|GetSpawnGhostTimer))))))

(event Custom|DemonPawnDied
  (params (DamagedActor :type Actor%20Object%20Reference) (Damage :type Float%20%28single-precision%29) (InstigatedBy :type Controller%20Object%20Reference) (DamageCauser :type Actor%20Object%20Reference) (HitLocation :type Vector) (FHitComponent :type Primitive%20Component%20Object%20Reference) (BoneName :type Name) (Momentum :type Vector))
  (Utilities|Time|ClearandInvalidateTimerbyHandle (Variables|Default|GetSpawnGhostTimer)))

(event Custom|TryToSpawnGhost
  (bind _hauntedplayers (Variables|Default|GetHauntedPlayers))
  (bind _hauntedplayers_2 (Variables|Default|GetHauntedPlayers))
  (bind _suitableplayer (Variables|Default|GetSuitablePlayer))
  (Variables|Default|SetSuitablePlayer 0)
  (bind _array_element (Utilities|Array|ForEachLoopwithBreak (Collision|GetOverlappingActors (Variables|/PuppeteerBoss/Box/Puppeteer_GhostSpawnArea_Box.Puppeteer_GhostSpawnArea_Box_C|GetBox (Variables|Default|GetGhostSpawnArea)) "/Script/FortniteGame.FortPlayerPawn" :!class /Script/Engine.PrimitiveComponent))
    (:LoopBody
      (bind _asfort_player_pawn (Utilities|Casting|CastToFortPlayerPawn _array_element))
      (if (Math|Boolean|ORBoolean (GameplayTags|HasAnyMatchingGameplayTags _asfort_player_pawn (GameplayTags|MakeLiteralGameplayTagContainer "(GameplayTags=((TagName=\"PuppeteerBoss.Player.IsHaunted\"),(TagName=\"PuppeteerBoss.Player.HauntCooldown\"),(TagName=\"Gameplay.Status.DBNO\")))")) (== (Utilities|GetClass _array_element) "/PuppeteerBoss/Pawn/BP_PlayerPawn_Puppeteer_Ghost.BP_PlayerPawn_Puppeteer_Ghost_C") (== (Utilities|GetClass _array_element) "/PuppeteerBoss/Pawn/BP_PlayerPawn_Puppeteer_Demon.BP_PlayerPawn_Puppeteer_Demon_C"))
        (if (GameplayTags|HasAnyMatchingGameplayTags _asfort_player_pawn (GameplayTags|MakeGameplayTagContainerfromTag "(TagName=\"PuppeteerBoss.Player.IsHaunted\")"))
          (if (Utilities|Array|ContainsItem _hauntedplayers _asfort_player_pawn)
            (Utilities|IsValid (Utilities|Array|Get(acopy) (Variables|Default|GetSpawnedGhosts) (Utilities|Array|FindItem _hauntedplayers _asfort_player_pawn))
              (:Is%20Not%20Valid
                (Class|BPFLAICommonFunctions|RemoveAllInstancesOfGameplayTag _asfort_player_pawn "(TagName=\"PuppeteerBoss.Player.IsHaunted\")")
                (Utilities|Array|RemoveIndex (Variables|Default|GetSpawnedGhosts) (Utilities|Array|FindItem _hauntedplayers _asfort_player_pawn))
                (Utilities|Array|RemoveIndex (Variables|Default|GetHauntedPlayers) (Utilities|Array|FindItem _hauntedplayers _asfort_player_pawn))))))
        (else
          (Variables|Default|SetSuitablePlayer _asfort_player_pawn)
          (CallFunction|FoundSuitablePlayer))))
    (:Completed
      (Utilities|IsValid _suitableplayer
        (:Is%20Valid
          (if (Utilities|Array|ContainsItem _hauntedplayers_2 _suitableplayer)
            (Utilities|IsValid (Utilities|Array|Get(acopy) (Variables|Default|GetSpawnedGhosts) (Utilities|Array|FindItem _hauntedplayers_2 _suitableplayer))
              (:Is%20Not%20Valid
                _))
            (else
              _))
          (bind _returnvalue (AI|EQS|RunEQSQuery "/PuppeteerBoss/EQS/EQS_GhostSpawnLocation.EQS_GhostSpawnLocation" self "RandomBest5Pct"))
          (AI|EQS|BindEventtoOnQueryFinished _returnvalue (delegate Custom|GhostSpawnQueryFinished)))))))

(event Custom|FoundSuitablePlayer
  (bind _hauntedplayers (Variables|Default|GetHauntedPlayers))
  (bind _hauntedplayers_2 (Variables|Default|GetHauntedPlayers))
  (bind _suitableplayer (Variables|Default|GetSuitablePlayer))
  (bind _array_element (Utilities|Array|ForEachLoopwithBreak (Collision|GetOverlappingActors (Variables|/PuppeteerBoss/Box/Puppeteer_GhostSpawnArea_Box.Puppeteer_GhostSpawnArea_Box_C|GetBox (Variables|Default|GetGhostSpawnArea)) "/Script/FortniteGame.FortPlayerPawn" :!class /Script/Engine.PrimitiveComponent))
    (:LoopBody
      (bind _asfort_player_pawn (Utilities|Casting|CastToFortPlayerPawn _array_element))
      (if (Math|Boolean|ORBoolean (GameplayTags|HasAnyMatchingGameplayTags _asfort_player_pawn (GameplayTags|MakeLiteralGameplayTagContainer "(GameplayTags=((TagName=\"PuppeteerBoss.Player.IsHaunted\"),(TagName=\"PuppeteerBoss.Player.HauntCooldown\"),(TagName=\"Gameplay.Status.DBNO\")))")) (== (Utilities|GetClass _array_element) "/PuppeteerBoss/Pawn/BP_PlayerPawn_Puppeteer_Ghost.BP_PlayerPawn_Puppeteer_Ghost_C") (== (Utilities|GetClass _array_element) "/PuppeteerBoss/Pawn/BP_PlayerPawn_Puppeteer_Demon.BP_PlayerPawn_Puppeteer_Demon_C"))
        (if (GameplayTags|HasAnyMatchingGameplayTags _asfort_player_pawn (GameplayTags|MakeGameplayTagContainerfromTag "(TagName=\"PuppeteerBoss.Player.IsHaunted\")"))
          (if (Utilities|Array|ContainsItem _hauntedplayers _asfort_player_pawn)
            (Utilities|IsValid (Utilities|Array|Get(acopy) (Variables|Default|GetSpawnedGhosts) (Utilities|Array|FindItem _hauntedplayers _asfort_player_pawn))
              (:Is%20Not%20Valid
                (Class|BPFLAICommonFunctions|RemoveAllInstancesOfGameplayTag _asfort_player_pawn "(TagName=\"PuppeteerBoss.Player.IsHaunted\")")
                (Utilities|Array|RemoveIndex (Variables|Default|GetSpawnedGhosts) (Utilities|Array|FindItem _hauntedplayers _asfort_player_pawn))
                (Utilities|Array|RemoveIndex (Variables|Default|GetHauntedPlayers) (Utilities|Array|FindItem _hauntedplayers _asfort_player_pawn))))))
        (else
          (Variables|Default|SetSuitablePlayer _asfort_player_pawn)
          (CallFunction|FoundSuitablePlayer))))
    (:Completed
      (Utilities|IsValid _suitableplayer
        (:Is%20Valid
          (if (Utilities|Array|ContainsItem _hauntedplayers_2 _suitableplayer)
            (Utilities|IsValid (Utilities|Array|Get(acopy) (Variables|Default|GetSpawnedGhosts) (Utilities|Array|FindItem _hauntedplayers_2 _suitableplayer))
              (:Is%20Not%20Valid
                _))
            (else
              _))
          (bind _returnvalue (AI|EQS|RunEQSQuery "/PuppeteerBoss/EQS/EQS_GhostSpawnLocation.EQS_GhostSpawnLocation" self "RandomBest5Pct"))
          (AI|EQS|BindEventtoOnQueryFinished _returnvalue (delegate Custom|GhostSpawnQueryFinished)))))))

(event Custom|GhostSpawnQueryFinished
  (params (QueryInstance :type EQS%20Query%20Instance%20Object%20Reference) (QueryStatus :type EEnvQueryStatus%20Enum :default "Processing"))
  (bind _ghostspawnarea (Variables|Default|GetGhostSpawnArea))
  (bind _suitableplayer (Variables|Default|GetSuitablePlayer))
  (bind _demonpawn (Variables|Default|GetDemonPawn))
  (bind (_resultlocations _returnvalue) (AI|EQS|GetQueryResultsasLocations QueryInstance))
  (if _returnvalue
    (bind _returnvalue_2 (Game|SpawnActorfromClass "/PuppeteerBoss/PointProvider/Puppeteer_Ghost_LWStaticPointProvider.Puppeteer_Ghost_LWStaticPointProvider_C" (Math|Transform|MakeTransform (Utilities|Array|Get(acopy) _resultlocations)) "Undefined" "MultiplyWithRoot" self 0))
    (Utilities|IsValid _suitableplayer
      (:Is%20Valid
        (Variables|/PuppeteerBoss/PointProvider/Puppeteer_Ghost_LWStaticPointProvider.Puppeteer_Ghost_LWStaticPointProvider_C|SetSuitablePlayer _suitableplayer _returnvalue_2)
        (Utilities|IsValid _demonpawn
          (:Is%20Valid
            (Variables|/PuppeteerBoss/PointProvider/Puppeteer_Ghost_LWStaticPointProvider.Puppeteer_Ghost_LWStaticPointProvider_C|SetDemonPawn _demonpawn _returnvalue_2)
            (Utilities|IsValid _ghostspawnarea
              (:Is%20Valid
                (Variables|/PuppeteerBoss/PointProvider/Puppeteer_Ghost_LWStaticPointProvider.Puppeteer_Ghost_LWStaticPointProvider_C|SetGhost%20Spawn%20Area _ghostspawnarea _returnvalue_2))
              (:Is%20Not%20Valid
                _))
            (EventDispatchers|BindEventtoOnActorSpawned _returnvalue_2 (delegate Custom|GhostSpawned)))
          (:Is%20Not%20Valid
            _)))
      (:Is%20Not%20Valid
        _))
    (Actor|DestroyActor _returnvalue_2)))

(event Custom|GhostSpawned
  (params (SpawnedActor :type Actor%20Object%20Reference))
  (bind _hauntedplayers (Variables|Default|GetHauntedPlayers))
  (Utilities|IsValid (stale Custom|GhostSpawnQueryFinished _returnvalue_2)
    (:Is%20Valid
      (EventDispatchers|UnbindEventfromOnActorSpawned (stale Custom|GhostSpawnQueryFinished _returnvalue_2) (delegate Custom|GhostSpawned)))
    (:Is%20Not%20Valid
      _))
  (bind _asbp_player_pawn_puppeteer_ghost (Utilities|Casting|CastToBP_PlayerPawn_Puppeteer_Ghost SpawnedActor))
  (bind _returnvalue (Utilities|Array|Add (Variables|Default|GetSpawnedGhosts) _asbp_player_pawn_puppeteer_ghost))
  (Utilities|IsValid (Utilities|Array|Get(acopy) _hauntedplayers _returnvalue)
    (:Is%20Valid
      (if (GameplayTags|HasMatchingGameplayTag (Utilities|Array|Get(acopy) _hauntedplayers _returnvalue) "(TagName=\"PuppeteerBoss.Player.IsHaunted\")")
        (Class|BPFLAICommonFunctions|RemoveAllInstancesOfGameplayTag (Utilities|Array|Get(acopy) _hauntedplayers _returnvalue) "(TagName=\"PuppeteerBoss.Player.IsHaunted\")")
        (else
          _))
      (Utilities|Array|SetArrayElem _hauntedplayers _returnvalue (Variables|Default|GetSuitablePlayer) true))
    (:Is%20Not%20Valid
      _)))

(fn ConstructionScript)
